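# Animation Blueprint for the female Fortnite DJ backpack, animating its disco ball and dangling parts while tracking the wearer's kill count.
# The input pose gets an additive BallRotator rotation on the disco ball bone, then RigidBody simulation relative to the backpack root and spring-constrained AnimDynamics on dyn_line.
# On initialization it watches the pawn's AthenaKills stat and stores changes in KillCount; the graphs shown never read KillCount or set BallRotator.
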AnimBlueprint /BRCosmetics/Accessories/FORT_Backpacks/Backpack_Female_Fortnite_DJ/Meshes/F_MED_Fortnite_DJ_AnimBP
target_skeleton=&/F_MED_Fortnite_DJ_Skeleton
graphs [
 AnimGraph {nodes [
   AnimGraphNode_SubInput_0 = Input Pose()
   n1 = "Transform (Modify) Bone - Bone: dyn_disco_ball"(
    ComponentPose=Local To Component(LocalPose=@AnimGraphNode_SubInput_0)
    Alpha=1
    Translation="0,0,0"
    Rotation=Get BallRotator() -> (BallRotator:Rotator)
    Scale="1,1,1"
   ) {rotationMode=BMM_Additive, rotationSpace=BCS_BoneSpace, alphaBoolBlend {blendOption=Linear}, boneToModify {boneName=dyn_disco_ball}}
   n2 = RigidBody(ComponentPose=@n1) {
    bOverrideWorldGravity=true
    simulationSpace=BaseBoneSpace
    alphaBoolBlend {blendOption=Linear}
    baseBoneRef {boneName=backpack_root}
    componentLinearAccScale=<0.5 0.5 0.5>
    componentLinearVelScale=<0.5 0.5 0.5>
    overrideWorldGravity {z=-980}
   }
   n3 = Anim Dynamics - Bone: dyn_line(ComponentPose=@n2, Alpha=1, ExternalForce="0,0,0") {
    angularSpringConstant=1000
    bAngularSpring=true
    bGravityOverrideInSimSpace=true
    bUseGravityOverride=true
    alphaBoolBlend {blendOption=Linear}
    boundBone {boneName=dyn_line}
    componentAppliedLinearAccClamp=<50 50 50>
    componentLinearAccScale=<1 1 1>
    componentLinearVelScale=<1 1 1>
    gravityOverride {z=-980}
    physicsBodyDefinitions [{
      sphereCollisionRadius=0
      boundBone {boneName=dyn_line}
      boxExtents=<10 2 10>
      constraintSetup {
       angularTargetAxis=AxisY
       twistAxis=AxisY
       angularLimitsMax=<12 2 -1>
       angularLimitsMin=<-12 -2 1>
       angularTarget=<0 0 0>
       linearAxesMax=<2 3 1>
       linearAxesMin=<-1 -3 -1>
      }
      localJointOffset=<0 -1 0>
     }]
   }
   Output Pose(Result=Component To Local(ComponentPose=@n3)) {layerGroup=None}
  ]}
 EventGraph {nodes [
   Event BlueprintInitializeAnimation()
   > n1 = Cast To FortPlayerPawn(Object=TryGetFortPawn())
   > n2 = WatchCosmeticStat!(
    execute:Exec
    Pawn:Fort Player Pawn Object Reference=@n1
    StatTag:Fort Stat Manager Tag Structure=(TagName="StatManager.AthenaKills")
    DefaultValue:int=0
    bBroadcastInitialValue:bool=true
    InitialValueBroadcastDelay:float=-1
    ChangeValueBroadcastDelay:float=-1
    bConsumeWaitingNotifications:bool=true
   ) -> (then:Exec, OnStatChanged:Exec, StatName:Name, StatValue:int) {is=title K2Node_AsyncAction}
   Set KillCount(execute=@n2.OnStatChanged, KillCount:int=@n2.StatValue) -> (Output_Get:int=0)
  ]}
]
variables []
